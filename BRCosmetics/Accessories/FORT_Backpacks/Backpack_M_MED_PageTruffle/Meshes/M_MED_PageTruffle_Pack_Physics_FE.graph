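# Physics asset for the PageTruffle backpack cosmetic, giving collision and joints to its head, toilet cap and toilet seat bones.
# Small capsules sit on the bp_ bones, while dyn_head gets one larger capsule and dyn_toilet_cap three parallel ones.
# The cap joins its bp_ bone with only Swing2 unlocked, acting as a hinge.
# The head joint locks all angular motion, and the seat has no constraint.
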PhysicsAsset /BRCosmetics/Accessories/FORT_Backpacks/Backpack_M_MED_PageTruffle/Meshes/M_MED_PageTruffle_Pack_Physics_FE
bodies {
 bp_head [{shapeType=Capsule, length=0.505, radius=0.505, shapeName=bp_head_capsule, rotation=<rot 0 -90 0>}]
 bp_toilet_cap [{shapeType=Capsule, length=0.505, radius=0.505, shapeName=bp_toilet_cap_capsule, rotation=<rot 0 -90 0>}]
 bp_toilet_seat [{shapeType=Capsule, length=0.505, radius=0.505, shapeName=bp_toilet_seat_capsule, rotation=<rot 0 -90 0>}]
 dyn_head [{shapeType=Capsule, length=0.505, radius=4.7, shapeName=dyn_head_capsule, center=<4.3557 -0.1201 0>, rotation=<rot 0 -90 0>}]
 dyn_toilet_cap [
  {shapeType=Capsule, length=10, radius=1.5, shapeName=dyn_toilet_cap_capsule, center=<7.1753 0.0018 1.4938>, rotation=<rot 0 90.0032 -86.7665>}
  {shapeType=Capsule, length=10, radius=1.5, shapeName=dyn_toilet_cap_capsule, center=<7.1753 3.3 1.4938>, rotation=<rot 0 90.0032 -86.7663>}
  {shapeType=Capsule, length=10, radius=1.5, shapeName=dyn_toilet_cap_capsule, center=<7.1753 -3.3 1.4938>, rotation=<rot 0 90.0032 -86.7663>}
 ]
}
constraints [
 {
  bone1Name=dyn_toilet_cap
  bone2Name=bp_toilet_cap
  swing1LimitDegrees=45
  swing1Motion=Locked
  swing2LimitDegrees=10
  twistLimitDegrees=45
  twistMotion=Locked
 }
 {
  bone1Name=dyn_head
  bone2Name=bp_head
  swing1LimitDegrees=45
  swing1Motion=Locked
  swing2LimitDegrees=45
  swing2Motion=Locked
  twistLimitDegrees=45
  twistMotion=Locked
 }
]
