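# Animation Blueprint for the PageTruffle backpack cosmetic, playing idles and reacting to enemy eliminations and chest openings.
# A state machine cycles four idles in fixed order (the random player is unconnected) and switches to reactions on anim input events.
# Its pose feeds AnimDynamics on the toilet seat bone, then RigidBody physics.
# Updates compute per-movement-state gravity and rigid-body settings; frontend reactivity replays elimination on a delay loop.
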AnimBlueprint /BRCosmetics/Accessories/FORT_Backpacks/Backpack_M_MED_PageTruffle/Meshes/M_MED_PageTruffle_Pack_AnimBP
target_skeleton=&/M_MED_PageTruffle_Pack_Skeleton
graphs [
 AnimGraph {nodes [
   Input Pose()
   AnimGraphNode_StateMachine_1 = Backpack States() {is=title AnimGraphNode_StateMachine}
   n1 = Anim Dynamics - Bone: bp_toilet_seat!(ComponentPose=Local To Component(LocalPose=@AnimGraphNode_StateMachine_1), Alpha=1, GravityOverride:Vector="0,0,-980", ExternalForce="0,0,0") -> (Pose) {
    angularDampingOverride=1
    angularSpringConstant=20000
    bAngularSpring=true
    bOverrideAngularDamping=true
    bUseGravityOverride=true
    gravityScale=0.35
    simulationSpace=RootRelative
    alphaBoolBlend {blendOption=Linear}
    boundBone {boneName=bp_toilet_seat}
    componentAppliedLinearAccClamp=<6000 6000 6000>
    componentLinearAccScale=<0.4 0.4 0.4>
    componentLinearVelScale=<1 1 1>
    gravityOverride {z=-980}
    physicsBodyDefinitions [{
      boundBone {boneName=bp_toilet_seat}
      boxExtents=<20 10 10>
      constraintSetup {angularLimitsMax=<0 0 0>, angularLimitsMin=<0 -7 0>, angularTarget=<1 0 0>}
      localJointOffset=<2 0 0>
     }]
   }
   n2 = RigidBody!(
    ComponentPose=@n1
    Alpha
    OverrideWorldGravity:Vector="0,0,0"
    ExternalForce
    SimSpaceSettings:Sim Space Settings Structure=(WorldAlpha=0,VelocityScaleZ=1,DampingAlpha=1,MaxLinearVelocity=10000,MaxAngularVelocity=10000,MaxLinearAcceleration=10000,MaxAngularAcceleration=10000,ExternalLinearDragV=(X=0,Y=0,Z=0),ExternalLinearVelocity=(X=0,Y=0,Z=0),ExternalAngularVelocity=(X=0,Y=0,Z=0))
   ) -> (Pose) {
    bOverrideWorldGravity=true
    overridePhysicsAsset=&/M_MED_PageTruffle_Pack_Physics_FE
    simulationSpace=BaseBoneSpace
    alphaBoolBlend {blendOption=Linear}
    baseBoneRef {boneName=pack}
    componentAppliedLinearAccClamp=<3000 3000 3000>
   }
   Output Pose(Result=Component To Local(ComponentPose=@n2)) {layerGroup=None}
  ]}
 Backpack States {
  entry=Locomotion
  path=AnimGraph.AnimGraphNode_StateMachine_1.Backpack States
  ids=id
  nodes [
   AnimStateNode_0 = Locomotion() {is=title AnimStateNode, stateEntered {guid=#1, notifyName=Entered_RandomLocoBreak}}
   AnimStateNode_4 = Elimination() {is=title AnimStateNode, stateEntered {guid=#2, notifyName=ElimEntered}}
   AnimStateNode_5 = OpenChest() {is=title AnimStateNode}
   AnimStateTransitionNode_11 = AnimStateTransitionNode()
   AnimStateTransitionNode_5 = AnimStateTransitionNode() {bAutomaticRuleBasedOnSequencePlayerInState=true}
   AnimStateTransitionNode_7 = AnimStateTransitionNode()
   AnimStateTransitionNode_8 = AnimStateTransitionNode() {bAutomaticRuleBasedOnSequencePlayerInState=true}
  ]
  links [
   {
    dst=Locomotion
    rule=AnimGraph.AnimGraphNode_StateMachine_1.Backpack States.AnimStateTransitionNode_5.AnimationTransitionGraph_0
    src=Elimination
    via=AnimStateTransitionNode_5
   }
   {
    dst=Elimination
    rule=AnimGraph.AnimGraphNode_StateMachine_1.Backpack States.AnimStateTransitionNode_11.AnimationTransitionGraph_0
    src=Locomotion
    via=AnimStateTransitionNode_11
   }
   {
    dst=OpenChest
    rule=AnimGraph.AnimGraphNode_StateMachine_1.Backpack States.AnimStateTransitionNode_7.AnimationTransitionGraph_0
    src=Locomotion
    via=AnimStateTransitionNode_7
   }
   {
    dst=Locomotion
    rule=AnimGraph.AnimGraphNode_StateMachine_1.Backpack States.AnimStateTransitionNode_8.AnimationTransitionGraph_0
    src=OpenChest
    via=AnimStateTransitionNode_8
   }
  ]
 }
 Locomotion {
  path=AnimGraph.AnimGraphNode_StateMachine_1.Backpack States.AnimStateNode_0.Locomotion
  nodes [
   Random Sequence Player!() -> (Pose:Pose Link Structure) {
    is=title AnimGraphNode_RandomPlayer
    entries [
     {
      chanceToPlay=0.25
      maxLoopCount=0
      maxPlayRate=1
      minLoopCount=0
      minPlayRate=1
      sequence=/BRCosmetics/Accessories/FORT_Backpacks/Backpack_M_MED_PageTruffle/Anim/PageTruffle_Idle1_SpinLeftRight_Anim_PACK
      blendIn {blendOption=Linear, blendTime=0.2, customCurve=null}
     }
     {
      chanceToPlay=0.25
      maxLoopCount=0
      maxPlayRate=1
      minLoopCount=0
      minPlayRate=1
      sequence=/BRCosmetics/Accessories/FORT_Backpacks/Backpack_M_MED_PageTruffle/Anim/PageTruffle_Idle2_FranticLookAround_Anim_PACK
      blendIn {blendOption=Linear, blendTime=0.2, customCurve=null}
     }
     {
      chanceToPlay=0.25
      maxLoopCount=0
      maxPlayRate=1
      minLoopCount=0
      minPlayRate=1
      sequence=/BRCosmetics/Accessories/FORT_Backpacks/Backpack_M_MED_PageTruffle/Anim/PageTruffle_Idle3_Buzz_Anim_PACK
      blendIn {blendOption=Linear, blendTime=0.2, customCurve=null}
     }
     {
      chanceToPlay=0.25
      maxLoopCount=0
      maxPlayRate=1
      minLoopCount=0
      minPlayRate=1
      sequence=/BRCosmetics/Accessories/FORT_Backpacks/Backpack_M_MED_PageTruffle/Anim/PageTruffle_Idle4_LookAround_Anim_PACK
      blendIn {blendOption=Linear, blendTime=0.2, customCurve=null}
     }
    ]
   }
   AnimGraphNode_StateMachine_0 = Animation Testing() {is=title AnimGraphNode_StateMachine}
   Output Animation Pose(Result=@AnimGraphNode_StateMachine_0)
  ]
 }
 Animation Testing {
  entry=Idle_4
  path=AnimGraph.AnimGraphNode_StateMachine_1.Backpack States.AnimStateNode_0.Locomotion.AnimGraphNode_StateMachine_0.Animation Testing
  ids=id
  nodes [
   AnimStateNode_0 = Idle_1() {
    is=title AnimStateNode
    stateEntered {guid=#3}
    stateFullyBlended {guid=#4}
    stateLeft {guid=#5}
   }
   AnimStateNode_1 = Idle_4() {
    is=title AnimStateNode
    stateEntered {guid=#6}
    stateFullyBlended {guid=#7}
    stateLeft {guid=#8}
   }
   AnimStateNode_2 = Idle_2() {
    is=title AnimStateNode
    stateEntered {guid=#6}
    stateFullyBlended {guid=#7}
    stateLeft {guid=#8}
   }
   AnimStateNode_3 = Idle_3() {
    is=title AnimStateNode
    stateEntered {guid=#6}
    stateFullyBlended {guid=#7}
    stateLeft {guid=#8}
   }
   AnimStateTransitionNode_0 = AnimStateTransitionNode() {
    bAutomaticRuleBasedOnSequencePlayerInState=true
    transitionEnd {guid=#9}
    transitionInterrupt {guid=#10}
    transitionStart {guid=#11}
   }
   AnimStateTransitionNode_11 = AnimStateTransitionNode() {bAutomaticRuleBasedOnSequencePlayerInState=true}
   AnimStateTransitionNode_12 = AnimStateTransitionNode() {bAutomaticRuleBasedOnSequencePlayerInState=true}
   AnimStateTransitionNode_13 = AnimStateTransitionNode() {bAutomaticRuleBasedOnSequencePlayerInState=true}
  ]
  links [
   {
    dst=Idle_4
    rule=AnimGraph.AnimGraphNode_StateMachine_1.Backpack States.AnimStateNode_0.Locomotion.AnimGraphNode_StateMachine_0.Animation Testing.AnimStateTransitionNode_13.Transition
    src=Idle_1
    via=AnimStateTransitionNode_13
   }
   {
    dst=Idle_3
    rule=AnimGraph.AnimGraphNode_StateMachine_1.Backpack States.AnimStateNode_0.Locomotion.AnimGraphNode_StateMachine_0.Animation Testing.AnimStateTransitionNode_0.Transition
    src=Idle_2
    via=AnimStateTransitionNode_0
   }
   {
    dst=Idle_1
    rule=AnimGraph.AnimGraphNode_StateMachine_1.Backpack States.AnimStateNode_0.Locomotion.AnimGraphNode_StateMachine_0.Animation Testing.AnimStateTransitionNode_12.Transition
    src=Idle_3
    via=AnimStateTransitionNode_12
   }
   {
    dst=Idle_2
    rule=AnimGraph.AnimGraphNode_StateMachine_1.Backpack States.AnimStateNode_0.Locomotion.AnimGraphNode_StateMachine_0.Animation Testing.AnimStateTransitionNode_11.Transition
    src=Idle_4
    via=AnimStateTransitionNode_11
   }
  ]
 }
 Idle_1 {
  path=AnimGraph.AnimGraphNode_StateMachine_1.Backpack States.AnimStateNode_0.Locomotion.AnimGraphNode_StateMachine_0.Animation Testing.AnimStateNode_0.Idle_1
  nodes [
   n1 = Sequence Player 'PageTruffle_Idle1_SpinLeftRight_Anim_PACK'() {
    bLoopAnimation=false
    sequence=&PageTruffle_Idle1_SpinLeftRight_Anim_PACK
   }
   Output Animation Pose(Result=@n1)
  ]
 }
 Idle_4 {
  path=AnimGraph.AnimGraphNode_StateMachine_1.Backpack States.AnimStateNode_0.Locomotion.AnimGraphNode_StateMachine_0.Animation Testing.AnimStateNode_1.Idle_4
  nodes [
   n1 = Sequence Player 'PageTruffle_Idle4_LookAround_Anim_PACK'() {
    bLoopAnimation=false
    sequence=&PageTruffle_Idle4_LookAround_Anim_PACK
   }
   Output Animation Pose(Result=@n1)
  ]
 }
 Idle_2 {
  path=AnimGraph.AnimGraphNode_StateMachine_1.Backpack States.AnimStateNode_0.Locomotion.AnimGraphNode_StateMachine_0.Animation Testing.AnimStateNode_2.Idle_2
  nodes [
   n1 = Sequence Player 'PageTruffle_Idle2_FranticLookAround_Anim_PACK'() {
    bLoopAnimation=false
    sequence=&PageTruffle_Idle2_FranticLookAround_Anim_PACK
   }
   Output Animation Pose(Result=@n1)
  ]
 }
 Idle_3 {
  path=AnimGraph.AnimGraphNode_StateMachine_1.Backpack States.AnimStateNode_0.Locomotion.AnimGraphNode_StateMachine_0.Animation Testing.AnimStateNode_3.Idle_3
  nodes [
   n1 = Sequence Player 'PageTruffle_Idle3_Buzz_Anim_PACK'() {bLoopAnimation=false, sequence=&PageTruffle_Idle3_Buzz_Anim_PACK}
   Output Animation Pose(Result=@n1)
  ]
 }
 Transition {
  path=AnimGraph.AnimGraphNode_StateMachine_1.Backpack States.AnimStateNode_0.Locomotion.AnimGraphNode_StateMachine_0.Animation Testing.AnimStateTransitionNode_0.Transition
  nodes [Result()]
 }
 Transition {
  path=AnimGraph.AnimGraphNode_StateMachine_1.Backpack States.AnimStateNode_0.Locomotion.AnimGraphNode_StateMachine_0.Animation Testing.AnimStateTransitionNode_11.Transition
  nodes [Result()]
 }
 Transition {
  path=AnimGraph.AnimGraphNode_StateMachine_1.Backpack States.AnimStateNode_0.Locomotion.AnimGraphNode_StateMachine_0.Animation Testing.AnimStateTransitionNode_12.Transition
  nodes [Result()]
 }
 Transition {
  path=AnimGraph.AnimGraphNode_StateMachine_1.Backpack States.AnimStateNode_0.Locomotion.AnimGraphNode_StateMachine_0.Animation Testing.AnimStateTransitionNode_13.Transition
  nodes [Result()]
 }
 Elimination {
  path=AnimGraph.AnimGraphNode_StateMachine_1.Backpack States.AnimStateNode_4.Elimination
  nodes [Output Animation Pose(Result=Sequence Player 'PageTruffle_Elim_Anim_PACK'() {bLoopAnimation=false, sequence=/BRCosmetics/Accessories/FORT_Backpacks/Backpack_M_MED_PageTruffle/Anim/PageTruffle_Elim_Anim_PACK})]
 }
 OpenChest {
  path=AnimGraph.AnimGraphNode_StateMachine_1.Backpack States.AnimStateNode_5.OpenChest
  nodes [
   n1 = Sequence Player 'PageTruffle_ChestReactive_Anim_PACK'() {bLoopAnimation=false, sequence=/BRCosmetics/Accessories/FORT_Backpacks/Backpack_M_MED_PageTruffle/Anim/PageTruffle_ChestReactive_Anim_PACK}
   Output Animation Pose(Result=@n1)
  ]
 }
 AnimationTransitionGraph_0 {
  path=AnimGraph.AnimGraphNode_StateMachine_1.Backpack States.AnimStateTransitionNode_11.AnimationTransitionGraph_0
  nodes [Result(bCanEnterTransition=Get bEnemyPlayerEliminated() -> (bEnemyPlayerEliminated:bool))]
 }
 AnimationTransitionGraph_0 {path=AnimGraph.AnimGraphNode_StateMachine_1.Backpack States.AnimStateTransitionNode_5.AnimationTransitionGraph_0, nodes [Result()]}
 AnimationTransitionGraph_0 {
  path=AnimGraph.AnimGraphNode_StateMachine_1.Backpack States.AnimStateTransitionNode_7.AnimationTransitionGraph_0
  nodes [Result(bCanEnterTransition=Get bChestOpened() -> (bChestOpened:bool))]
 }
 AnimationTransitionGraph_0 {path=AnimGraph.AnimGraphNode_StateMachine_1.Backpack States.AnimStateTransitionNode_8.AnimationTransitionGraph_0, nodes [Result()]}
 EventGraph {nodes [
   Event BlueprintUpdateAnimation()
   > n1 = "Gravity Override State Function (Nativized)"!(
    execute
    CharacterPartAnimInstance:Custom Character Part Anim Instance Object Reference=Self-Reference()
    GravityOverrideParameterSet:Fort Gravity Override Parameter Set Structure=Get Gravity Override Params_0() -> (Gravity Override Params_0:Fort Gravity Override Parameter Set Structure)
    RigidBodyAnimNodeParameterSet:Fort Rigid Body Anim Node Parameter Set Structure=Get RBAN Params_0() -> (RBAN Params_0:Fort Rigid Body Anim Node Parameter Set Structure)
    DebugDraw:bool=false
    GameTimeInSeconds:float=GetGameTimeInSeconds()
    DebugJointName:Name=root
    StateLogicLayerName:Name=none
   ) -> (
    then
    OutGravityOverride:Vector
    OutCurrentState:Name=None
    OutGravityOverrideParameters:Fort Gravity Override Parameters Structure
    OutRigidBodyAnimNodeParameters:Fort Rigid Body Anim Node Parameters Structure
   )
   > Set Gravity Override(Gravity Override=@n1)
   > Set RBAN Settings Out_0(RBAN Settings Out_0:Fort Rigid Body Anim Node Parameters Structure=@n1.OutRigidBodyAnimNodeParameters) -> (Output_Get:Fort Rigid Body Anim Node Parameters Structure)
   n2 = Event BlueprintInitializeAnimation()
   Branch(execute=[@n2.then, @n3.then], Condition=Get bShouldDisplayFrontendReactivity() -> (bShouldDisplayFrontendReactivity:bool))
   > Delay(Duration=20)
   > Set bEnemyPlayerEliminated(bEnemyPlayerEliminated:bool=true) -> (Output_Get:bool)
   Event BlueprintPostEvaluateAnimation() {is=title K2Node_Event}
   > Set bEnemyPlayerEliminated(bEnemyPlayerEliminated=false)
   > Set bChestOpened(bChestOpened:bool=false) -> (Output_Get:bool)
   AnimNotify_ElimEntered() {is=title K2Node_Event}
   > n3 = Set bEnemyPlayerEliminated(bEnemyPlayerEliminated=false)
   n4 = Event BlueprintProcessAnimInputEvent()
   > Sequence()
   > n5 = Branch(Condition=IsNamed(self=@n4.AnimInputEvent, InEventName=aie.EnemyPlayerEliminated))
   > Set bEnemyPlayerEliminated(bEnemyPlayerEliminated=true)
   Branch(execute=@n5.else, Condition=IsNamed(self=@n4.AnimInputEvent, InEventName=aie.ChestOpened))
   > Set bChestOpened(bChestOpened=true)
  ]}
 OnStateEntry {nodes [
   OnStateEntry!() -> (then, Context:"Anim Update Context Structure (by ref)", Node:"Anim Node Reference Structure (by ref)") {is=title K2Node_FunctionEntry}
   > Return Node()
  ]}
]
variables [
 bChestOpened {type=boolean}
 bEnemyPlayerEliminated {type=boolean}
 Gravity Override {type=Vector}
 Gravity Override Params_0 {
  type=FortGravityOverrideParameterSet
  default=(Skydive_GoingUp=(GravityOverride=(WindFrequency=(X=1.500000,Y=7.000000,Z=8.000000),WindAmplitude=(X=200.000000,Y=-500.000000,Z=400.000000),GravityOverride=(X=0.000000,Y=0.000000,Z=-980.000000),JointName="root")),Skydive_FloatingDown=(GravityOverride=(WindAmplitude=(X=0.000000,Y=-500.000000,Z=400.000000),WindOffset=(X=0.000000,Y=-600.000000,Z=400.000000),GravityOverride=(X=0.000000,Y=0.000000,Z=-980.000000),JointName="root")),Skydive_DivingDown=(GravityOverride=(WindFrequency=(X=1.500000,Y=12.000000,Z=10.000000),WindAmplitude=(X=50.000000,Y=300.000000,Z=100.000000),GravityOverride=(X=0.000000,Y=-50.000000,Z=-50.000000),JointName="pelvis")),Skydive_Parachuting=(GravityOverride=(WindFrequency=(X=1.500000,Y=7.000000,Z=8.000000),WindAmplitude=(X=200.000000,Y=-500.000000,Z=400.000000),GravityOverride=(X=0.000000,Y=-800.000000,Z=-980.000000),JointName="root")),OnGround_Standing=(GravityOverride=(GravityOverride=(X=0.000000,Y=0.000000,Z=-980.000000),JointName="root")),OnGround_Moving=(GravityOverride=(GravityOverride=(X=0.000000,Y=-500.000000,Z=-980.000000),JointName="root")),OnGround_Moving_HighRotationRate=(GravityOverride=(GravityOverride=(X=0.000000,Y=0.000000,Z=-600.000000),JointName="root")),OnGround_Crouch_Moving=(GravityOverride=(GravityOverride=(X=0.000000,Y=0.000000,Z=-980.000000),JointName="root")),Emote_or_Melee=(GravityOverride=(GravityOverride=(X=0.000000,Y=0.000000,Z=-980.000000),JointName="root")),Swimming=(GravityOverride=(GravityOverride=(X=-800.000000,Y=0.000000,Z=0.000000),JointName="pelvis")),Front_End=(GravityOverride=(GravityOverride=(X=0.000000,Y=0.000000,Z=-980.000000),JointName="root")),DBNO=(GravityOverride=(GravityOverride=(X=-980.000000,Y=400.000000,Z=0.000000),JointName="pelvis")),DBNO_Carried=(GravityOverride=(GravityOverride=(X=-980.000000,Y=500.000000,Z=-500.000000),JointName="pelvis")),OnGround_Falling=(GravityOverride=(GravityOverride=(X=0.000000,Y=0.000000,Z=-980.000000),JointName="root")),Ziplining=(GravityOverride=(GravityOverride=(X=0.000000,Y=0.000000,Z=-980.000000),JointName="root")),Skydive_Surfing=(GravityOverride=(WindFrequency=(X=1.500000,Y=7.000000,Z=8.000000),WindAmplitude=(X=200.000000,Y=-500.000000,Z=400.000000),GravityOverride=(X=500.000000,Y=-300.000000,Z=-980.000000),JointName="root")),Riding_Standing=(GravityOverride=(GravityOverride=(X=0.000000,Y=-350.000000,Z=-980.000000),JointName="root")),Riding_Moving=(GravityOverride=(GravityOverride=(X=0.000000,Y=-450.000000,Z=-450.000000),JointName="root")),Riding_Moving_Targeting=(GravityOverride=(GravityOverride=(X=0.000000,Y=-200.000000,Z=-980.000000),JointName="root")),Tactical_Sprint=(GravityOverride=(GravityOverride=(X=0.000000,Y=-850.000000,Z=1000.000000),JointName="root")),Rail_Grinding=(GravityOverride=(GravityOverride=(X=0.000000,Y=0.000000,Z=-980.000000),JointName="root")),)
 }
 RBAN Params_0 {
  type=FortRigidBodyAnimNodeParameterSet
  default=(Skydive_GoingUp=(RigidBodyAnimNode=(SimSpaceSettings=(WorldAlpha=1.000000,MaxLinearVelocity=500.000000,MaxAngularVelocity=3.000000,MaxLinearAcceleration=1000.000000,MaxAngularAcceleration=1.000000,ExternalLinearDragV=(X=1.000000,Y=1.000000,Z=1.000000)))),Skydive_FloatingDown=(RigidBodyAnimNode=(ComponentLinearAccScale=(X=0.000000,Y=0.500000,Z=0.000000),ComponentLinearVelScale=(X=0.000000,Y=0.500000,Z=0.000000),SimSpaceSettings=(WorldAlpha=1.000000,MaxLinearVelocity=500.000000,MaxAngularVelocity=3.000000,MaxLinearAcceleration=1000.000000,MaxAngularAcceleration=1.000000,ExternalLinearDragV=(X=1.000000,Y=1.000000,Z=1.000000)))),Skydive_DivingDown=(RigidBodyAnimNode=(ComponentLinearAccScale=(X=0.100000,Y=0.000000,Z=0.000000),ComponentLinearVelScale=(X=0.100000,Y=0.000000,Z=0.000000),SimSpaceSettings=(WorldAlpha=1.000000,MaxLinearVelocity=500.000000,MaxAngularVelocity=3.000000,MaxLinearAcceleration=1000.000000,MaxAngularAcceleration=1.000000,ExternalLinearDragV=(X=1.000000,Y=1.000000,Z=1.000000)))),Skydive_Parachuting=(RigidBodyAnimNode=(ComponentLinearAccScale=(X=0.000000,Y=0.100000,Z=0.300000),ComponentLinearVelScale=(X=0.000000,Y=0.100000,Z=0.300000),SimSpaceSettings=(WorldAlpha=1.000000,MaxLinearVelocity=500.000000,MaxAngularVelocity=3.000000,MaxLinearAcceleration=1000.000000,MaxAngularAcceleration=1.000000,ExternalLinearDragV=(X=1.000000,Y=1.000000,Z=1.000000)))),OnGround_Standing=(RigidBodyAnimNode=(ComponentLinearAccScale=(X=0.300000,Y=0.000000,Z=0.000000),ComponentLinearVelScale=(X=0.300000,Y=0.000000,Z=0.000000),SimSpaceSettings=(WorldAlpha=1.000000,DampingAlpha=0.250000,MaxLinearVelocity=500.000000,MaxAngularVelocity=20.000000,MaxLinearAcceleration=1000.000000,MaxAngularAcceleration=80.000000,ExternalLinearDragV=(X=1.000000,Y=1.000000,Z=1.000000)))),OnGround_Moving=(RigidBodyAnimNode=(ComponentLinearAccScale=(X=0.300000,Y=0.000000,Z=0.000000),ComponentLinearVelScale=(X=0.300000,Y=0.000000,Z=0.000000),SimSpaceSettings=(WorldAlpha=1.000000,DampingAlpha=0.250000,MaxLinearVelocity=500.000000,MaxAngularVelocity=20.000000,MaxLinearAcceleration=1000.000000,MaxAngularAcceleration=80.000000,ExternalLinearDragV=(X=1.000000,Y=1.000000,Z=1.000000)))),OnGround_Moving_HighRotationRate=(RigidBodyAnimNode=(ComponentLinearAccScale=(X=0.300000,Y=0.300000,Z=0.000000),ComponentLinearVelScale=(X=0.300000,Y=0.300000,Z=0.000000),SimSpaceSettings=(WorldAlpha=1.000000,DampingAlpha=0.500000,MaxLinearVelocity=500.000000,MaxAngularVelocity=3.000000,MaxLinearAcceleration=1000.000000,MaxAngularAcceleration=1.000000,ExternalLinearDragV=(X=1.000000,Y=1.000000,Z=1.000000)))),OnGround_Crouch_Moving=(RigidBodyAnimNode=(SimSpaceSettings=(WorldAlpha=1.000000,MaxLinearVelocity=500.000000,MaxAngularVelocity=3.000000,MaxLinearAcceleration=1000.000000,MaxAngularAcceleration=1.000000,ExternalLinearDragV=(X=1.000000,Y=1.000000,Z=1.000000)))),Emote_or_Melee=(RigidBodyAnimNode=(SimSpaceSettings=(WorldAlpha=1.000000,DampingAlpha=0.250000,MaxLinearVelocity=500.000000,MaxAngularVelocity=3.000000,MaxLinearAcceleration=1000.000000,MaxAngularAcceleration=1.000000,ExternalLinearDragV=(X=1.000000,Y=1.000000,Z=1.000000)))),Swimming=(RigidBodyAnimNode=(SimSpaceSettings=(WorldAlpha=1.000000,MaxLinearVelocity=500.000000,MaxAngularVelocity=3.000000,MaxLinearAcceleration=1000.000000,MaxAngularAcceleration=1.000000,ExternalLinearDragV=(X=1.000000,Y=1.000000,Z=1.000000)))),Front_End=(RigidBodyAnimNode=(SimSpaceSettings=(WorldAlpha=1.000000,MaxLinearVelocity=500.000000,MaxAngularVelocity=3.000000,MaxLinearAcceleration=1000.000000,MaxAngularAcceleration=1.000000,ExternalLinearDragV=(X=1.000000,Y=1.000000,Z=1.000000)))),DBNO=(RigidBodyAnimNode=(SimSpaceSettings=(WorldAlpha=1.000000,MaxLinearVelocity=500.000000,MaxAngularVelocity=3.000000,MaxLinearAcceleration=1000.000000,MaxAngularAcceleration=1.000000,ExternalLinearDragV=(X=1.000000,Y=1.000000,Z=1.000000)))),DBNO_Carried=(RigidBodyAnimNode=(SimSpaceSettings=(WorldAlpha=1.000000,MaxLinearVelocity=500.000000,MaxAngularVelocity=3.000000,MaxLinearAcceleration=1000.000000,MaxAngularAcceleration=1.000000,ExternalLinearDragV=(X=1.000000,Y=1.000000,Z=1.000000)))),OnGround_Falling=(RigidBodyAnimNode=(SimSpaceSettings=(WorldAlpha=1.000000,DampingAlpha=0.500000,MaxLinearVelocity=500.000000,MaxAngularVelocity=20.000000,MaxLinearAcceleration=1000.000000,MaxAngularAcceleration=80.000000,ExternalLinearDragV=(X=1.000000,Y=1.000000,Z=1.000000)))),Ziplining=(RigidBodyAnimNode=(SimSpaceSettings=(WorldAlpha=1.000000,MaxLinearVelocity=500.000000,MaxAngularVelocity=3.000000,MaxLinearAcceleration=1000.000000,MaxAngularAcceleration=1.000000,ExternalLinearDragV=(X=1.000000,Y=1.000000,Z=1.000000)))),Skydive_Surfing=(RigidBodyAnimNode=(SimSpaceSettings=(WorldAlpha=1.000000,MaxLinearVelocity=500.000000,MaxAngularVelocity=3.000000,MaxLinearAcceleration=1000.000000,MaxAngularAcceleration=1.000000,ExternalLinearDragV=(X=1.000000,Y=1.000000,Z=1.000000)))),Riding_Standing=(RigidBodyAnimNode=(SimSpaceSettings=(WorldAlpha=1.000000,MaxLinearVelocity=500.000000,MaxAngularVelocity=3.000000,MaxLinearAcceleration=1000.000000,MaxAngularAcceleration=1.000000,ExternalLinearDragV=(X=1.000000,Y=1.000000,Z=1.000000)))),Riding_Moving=(RigidBodyAnimNode=(SimSpaceSettings=(WorldAlpha=1.000000,MaxLinearVelocity=500.000000,MaxAngularVelocity=3.000000,MaxLinearAcceleration=1000.000000,MaxAngularAcceleration=1.000000,ExternalLinearDragV=(X=1.000000,Y=1.000000,Z=1.000000)))),Riding_Moving_Targeting=(RigidBodyAnimNode=(SimSpaceSettings=(WorldAlpha=1.000000,MaxLinearVelocity=500.000000,MaxAngularVelocity=3.000000,MaxLinearAcceleration=1000.000000,MaxAngularAcceleration=1.000000,ExternalLinearDragV=(X=1.000000,Y=1.000000,Z=1.000000)))),Tactical_Sprint=(RigidBodyAnimNode=(ComponentLinearAccScale=(X=0.300000,Y=0.000000,Z=0.000000),ComponentLinearVelScale=(X=0.300000,Y=0.000000,Z=0.000000),SimSpaceSettings=(WorldAlpha=1.000000,MaxLinearVelocity=500.000000,MaxAngularVelocity=3.000000,MaxLinearAcceleration=1000.000000,MaxAngularAcceleration=1.000000,ExternalLinearDragV=(X=1.000000,Y=1.000000,Z=1.000000)))),Rail_Grinding=(RigidBodyAnimNode=(SimSpaceSettings=(WorldAlpha=1.000000,MaxLinearVelocity=500.000000,MaxAngularVelocity=3.000000,MaxLinearAcceleration=1000.000000,MaxAngularAcceleration=1.000000,ExternalLinearDragV=(X=1.000000,Y=1.000000,Z=1.000000)))),)
 }
 RBAN Settings Out_0 {type=FortRigidBodyAnimNodeParameters}
]
